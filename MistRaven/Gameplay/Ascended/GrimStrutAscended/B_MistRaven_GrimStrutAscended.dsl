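; Weapon Blueprint for the Mist Raven GrimStrut Ascended, handling what happens when it is equipped.
; On equip and attach, it checks the instigator for the MistRaven.GrimStrut.Ascended.InEquipment tag.
; With the tag present it calls the parent logic and plays an attached equip sound; otherwise it only adds the tag.
; A note marks this as a fix for picking up a second Mist Raven while one is equipped.
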(event EventBeginPlay
  (|CallParentFunction) ; This node is disabled and will not be called.\nDrag off pins to build functionality.
)

(event EventTick
  (params (DeltaSeconds :type Float%20%28single-precision%29)))

(event FortWeapon|EventOnEquippedAndAttached
  (bind _returnvalue (Game|GetInstigator))
  ;[0.15 0.002451 0.03353 0.5] Solving Bug where if player tries to pick up a new Mist Raven w/ 1 already equipped
  ; Check and add tag "MistRaven.GrimStrut.Ascended.InEquipment" - Kentaro Tamai 10/25/25
  (bind _returnvalue_2 (GameplayTags|HasMatchingGameplayTag(Message) _returnvalue "(TagName=\"MistRaven.GrimStrut.Ascended.InEquipment\")"))
  (if _returnvalue_2
    (|CallParentFunction)
    (Audio|SpawnSoundAttached "/MistRaven/Assets/Audio/MSS_MistRaven_GrimStrut_Equip.MSS_MistRaven_GrimStrut_Equip" (Variables|Visual|GetWeaponMesh))
    (else
      (Ability|GameplayEffect|AddGameplayTags _returnvalue (Variables|Default|GetAddGrimStrutAscendedInEquipmentTag)))))

(fn ConstructionScript
  (|CallParentFunction))
