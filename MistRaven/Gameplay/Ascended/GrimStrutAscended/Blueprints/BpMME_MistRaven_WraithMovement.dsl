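; Fortnite movement mode extension for MistRaven's Grim Strut wraith flight, active while the owner has the GrimStrut.Active tag.
; After each move it stores input (requested velocity for bots), timing and control rotation in runtime data, then hovers above ground.
; PhysUpdate slerps velocity toward the desired direction, capping speed on a fast/slow boost curve, with Ascended variants and Z clamp.
; Hits use a selectable slide, normal-bump or ricochet strategy, each repeated up to three times.
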(event EnhancedInputActionIA_MistRaven_TempTestingAction
  (params (ActionValue :type Boolean) (ElapsedSeconds :type Float%20%28double-precision%29) (TriggeredSeconds :type Float%20%28double-precision%29) (InputAction :type Input%20Action%20Object%20Reference :default "/MistRaven/Gameplay/Ascended/GrimStrutAscended/Blueprints/Input/IA_MistRaven_TempTestingAction.IA_MistRaven_TempTestingAction")))

(event Custom|PushInputContext
  (params (Controller :type Controller%20Object%20Reference))
  ;[1 0 0] Input (WIP non-functioning without backing code changes to MME system)
  (bind _asfort_player_controller (Utilities|Casting|CastToFortPlayerController Controller))
  (if (not (Networking|HasAuthority _asfort_player_controller))
    (Input|AddMappingContext (PlayerController|LocalPlayerSubsystems|GetEnhancedInputLocalPlayerSubsystem _asfort_player_controller) (Variables|Default|GetMistRavenInputContext) 0 "(bIgnoreAllPressedKeysUntilRelease=True,bForceImmediately=False,bNotifyUserSettings=False)")))

(event EnhancedInputActionIA_MistRavenFlyUp
  (params (ActionValue :type Boolean) (ElapsedSeconds :type Float%20%28double-precision%29) (TriggeredSeconds :type Float%20%28double-precision%29) (InputAction :type Input%20Action%20Object%20Reference :default "/MistRaven/Gameplay/Ascended/GrimStrutAscended/Blueprints/Input/IA_MistRavenFlyUp.IA_MistRavenFlyUp")))

(event Custom|PopInputContext
  (params (Controller :type Controller%20Object%20Reference))
  ;[1 0 0] Input (WIP non-functioning without backing code changes to MME system)
  (bind _asfort_player_controller (Utilities|Casting|CastToFortPlayerController Controller))
  (if (not (Networking|HasAuthority _asfort_player_controller))
    (Input|RemoveMappingContext (PlayerController|LocalPlayerSubsystems|GetEnhancedInputLocalPlayerSubsystem _asfort_player_controller) 0 "(bIgnoreAllPressedKeysUntilRelease=True,bForceImmediately=False,bNotifyUserSettings=False)")))

(event EnhancedInputActionIA_MistRavenFlyDown
  (params (ActionValue :type Boolean) (ElapsedSeconds :type Float%20%28double-precision%29) (TriggeredSeconds :type Float%20%28double-precision%29) (InputAction :type Input%20Action%20Object%20Reference :default "/MistRaven/Gameplay/Ascended/GrimStrutAscended/Blueprints/Input/IA_MistRavenFlyDown.IA_MistRavenFlyDown")))

(event EventOnExtensionActivate
  (params (RuntimeContext :type MMERuntime%20Context%20Structure%20%28by%20ref%29))
  (bind _ownerpawn (Variables|FortMovementMode_BaseExtLogic|GetOwnerPawn))
  (bind _returnvalue (Actor|GetComponentbyClass _ownerpawn "/MistRaven/B_MistRaven_PlayerComponent.B_MistRaven_PlayerComponent_C"))
  ;[0.2541 1 0.2188] Extension Activate
  (Variables|Default|SetServerTotalTime 0.0) ; Reset if we start a MME.\nWe will likely need to make this\nmore robust, either on the MME,\nplayer component, or something else\nWe don't want to allow player spoofing time.\nBut we may need to handle coming back into\nwraith form 
  (Utilities|FlowControl|Sequence
    (:then_0
      (CallFunction|PushInputContext (Pawn|GetController (Variables|FortMovementMode_BaseExtLogic|GetOwnerPawn))))
    (:then_1
      ;[1 1 1] Handle no input held after dash, set velocity to something reasonable
      (if (< (Math|Vector2D|Vector2DLengthSquared (Math|Conversions|ToVector2D(Vector) (Transformation|GetVelocity (Variables|FortMovementMode_BaseExtLogic|GetOwnerPawn)))) 1.0)
        (Utilities|IsValid _returnvalue
          (:Is%20Valid
            (Variables|/Script/Engine.MovementComponent|SetVelocity (* (Math|Vector|Normalize (Math|Vector|MakeVector (.x (- (Transformation|GetActorLocation (Variables|FortMovementMode_BaseExtLogic|GetOwnerPawn)) (Variables|/MistRaven/B_MistRaven_PlayerComponent.B_MistRaven_PlayerComponent_C|GetStartGrimStrutLocation _returnvalue))) (.y (- (Transformation|GetActorLocation (Variables|FortMovementMode_BaseExtLogic|GetOwnerPawn)) (Variables|/MistRaven/B_MistRaven_PlayerComponent.B_MistRaven_PlayerComponent_C|GetStartGrimStrutLocation _returnvalue))) (Math|Float|Max(Float) (.z (- (Transformation|GetActorLocation (Variables|FortMovementMode_BaseExtLogic|GetOwnerPawn)) (Variables|/MistRaven/B_MistRaven_PlayerComponent.B_MistRaven_PlayerComponent_C|GetStartGrimStrutLocation _returnvalue)))))) (ScalableFloat|GetValueasDouble (Variables|Default|GetFastWraithFlightSpeed))) (Variables|/Script/Engine.Character|GetCharacterMovement (Variables|FortMovementMode_BaseExtLogic|GetOwnerPawn)))))))))

(event EventOnExtensionPush
  (params (RuntimeContext :type MMERuntime%20Context%20Structure%20%28by%20ref%29)))

(event EventOnExtensionDeactivate
  (params (RuntimeContext :type MMERuntime%20Context%20Structure%20%28by%20ref%29))
  ;[1 0.1562 0.1562] Extension Deactivate
  (CallFunction|PopInputContext (Pawn|GetController (Variables|FortMovementMode_BaseExtLogic|GetOwnerPawn))))

(event EventOnExtensionEnd
  (params (RuntimeContext :type MMERuntime%20Context%20Structure%20%28by%20ref%29)))

(event EventUpdateAfterCharacterMovement
  (params (RuntimeContext :type MMERuntime%20Context%20Structure%20%28by%20ref%29) (OldLocation :type Vector%20%28by%20ref%29 :default "0, 0, 0") (OldVelocity :type Vector%20%28by%20ref%29 :default "0, 0, 0"))
  (bind _ownerpawn (Variables|FortMovementMode_BaseExtLogic|GetOwnerPawn))
  (bind _returnvalue (Fortnite|GetFortPlayerControllerfromActor _ownerpawn))
  ;[0.2396 0.8092 1] UpdateAfterCharacterMovement
  (|CallParentFunction RuntimeContext OldLocation OldVelocity)
  (bind _returnvalue_2 (Class|FortMovementModeExtBlueprintLibrary|GetRuntimeDatafromContext RuntimeContext "/MistRaven/Gameplay/Ascended/GrimStrutAscended/Blueprints/BpMME_RuntimeData_MistRaven_WraithMovement.BpMME_RuntimeData_MistRaven_WraithMovement_C"))
  (Utilities|IsValid _returnvalue_2
    (:Is%20Valid
      (Utilities|FlowControl|Sequence
        (:then_0
          (Variables|/MistRaven/Gameplay/Ascended/GrimStrutAscended/Blueprints/BpMME_RuntimeData_MistRaven_WraithMovement.BpMME_RuntimeData_MistRaven_WraithMovement_C|SetControlRotation (Pawn|GetControlRotation (Variables|FortMovementMode_BaseExtLogic|GetOwnerPawn) :!class /Script/Engine.Pawn) _returnvalue_2)
          (if (Pawn|IsBotControlled (Variables|FortMovementMode_BaseExtLogic|GetOwnerPawn))
            (Variables|/MistRaven/Gameplay/Ascended/GrimStrutAscended/Blueprints/BpMME_RuntimeData_MistRaven_WraithMovement.BpMME_RuntimeData_MistRaven_WraithMovement_C|SetRequestedVelocity (Pawn|Components|CharacterMovement|GetLastUpdateRequestedVelocity (Variables|/Script/Engine.Character|GetCharacterMovement (Variables|FortMovementMode_BaseExtLogic|GetOwnerPawn))) _returnvalue_2)
            (else
              (Variables|/MistRaven/Gameplay/Ascended/GrimStrutAscended/Blueprints/BpMME_RuntimeData_MistRaven_WraithMovement.BpMME_RuntimeData_MistRaven_WraithMovement_C|SetInputVector (Class|FortMovementBlueprintLibrary|GetLocalMoveInputVector (Variables|FortMovementMode_BaseExtLogic|GetOwnerPawn)) _returnvalue_2)
              (Variables|/MistRaven/Gameplay/Ascended/GrimStrutAscended/Blueprints/BpMME_RuntimeData_MistRaven_WraithMovement.BpMME_RuntimeData_MistRaven_WraithMovement_C|SetUpInput (Class|FortMovementBlueprintLibrary|GetMMELocalLegacyJumpInputPressed _returnvalue) _returnvalue_2)
              (Variables|/MistRaven/Gameplay/Ascended/GrimStrutAscended/Blueprints/BpMME_RuntimeData_MistRaven_WraithMovement.BpMME_RuntimeData_MistRaven_WraithMovement_C|SetDownInput (Class|FortMovementBlueprintLibrary|GetMMELocalLegacyCrouchInputPressed _returnvalue) _returnvalue_2)))
          (Variables|/MistRaven/Gameplay/Ascended/GrimStrutAscended/Blueprints/BpMME_RuntimeData_MistRaven_WraithMovement.BpMME_RuntimeData_MistRaven_WraithMovement_C|SetTotalTimeUpdated (+ (Class|FortMovementModeExtBlueprintLibrary|GetDeltaTimefromContext RuntimeContext (Variables|FortMovementMode_BaseExtLogic|GetbUseDilatedDeltaTime)) (Variables|/MistRaven/Gameplay/Ascended/GrimStrutAscended/Blueprints/BpMME_RuntimeData_MistRaven_WraithMovement.BpMME_RuntimeData_MistRaven_WraithMovement_C|GetTotalTimeUpdated _returnvalue_2)) _returnvalue_2))
        (:then_1
          (CallFunction|DoHoverLogic RuntimeContext))))))

(event Custom|BumpUpSlide
  (params (MMERuntimeContext :type MMERuntime%20Context%20Structure))
  ;[0.4176 0.1208 0.5156] Handle Hit
  (Transformation|SetActorLocation (Variables|FortMovementMode_BaseExtLogic|GetOwnerPawn) (+ (Transformation|GetActorLocation (Variables|FortMovementMode_BaseExtLogic|GetOwnerPawn)) (Math|Vector|MakeVector 0.0 0.0 5.0)) true false) ; sweep so we get collision.\ntry to slide up if we can't slide in any direction\npreferably this would be the step up logic
)

(event EventHandleHitResponse
  (params (RuntimeContext :type MMERuntime%20Context%20Structure%20%28by%20ref%29) (InVelocity :type Vector%20%28by%20ref%29 :default "0, 0, 0") (OldPosition :type Vector%20%28by%20ref%29 :default "0, 0, 0") (NewLocation :type Vector%20%28by%20ref%29 :default "0, 0, 0") (Adjustement :type Vector%20%28by%20ref%29 :default "0, 0, 0") (Hit :type Hit%20Result%20Structure%20%28by%20ref%29))
  ;[0.4176 0.1208 0.5156] Handle Hit
  (switch int (ScalableFloat|GetValueasInteger (Variables|Default|GetHitStrategy))
    (:0
      (bind (_handledblockinghit _outvelocity _outoldposition _outnewposition _outnewadjustment _sweephitresult) (Default|HandleHitSlide InVelocity OldPosition NewLocation Adjustement Hit))
      (if _handledblockinghit
        (bind (_handledblockinghit_2 _outvelocity_2 _outoldposition_2 _outnewposition_2 _outnewadjustment_2 _sweephitresult_2) (Default|HandleHitSlide _outvelocity _outoldposition _outnewposition _outnewadjustment _sweephitresult 1))
        (if _handledblockinghit_2
          (bind _handledblockinghit_3 (Default|HandleHitSlide _outvelocity_2 _outoldposition_2 _outnewposition_2 _outnewadjustment_2 _sweephitresult_2 2))
          (if _handledblockinghit_3
            (CallFunction|BumpUpSlide RuntimeContext)))))
    (:1
      (bind (_handledblockinghit_4 _outvelocity_3 _outoldposition_3 _outnewposition_3 _outnewadjustment_3 _sweephitresult_3) (CallFunction|HandleHitNormalBump InVelocity OldPosition NewLocation Adjustement Hit))
      (if _handledblockinghit_4
        (bind (_handledblockinghit_5 _outvelocity_4 _outoldposition_4 _outnewposition_4 _outnewadjustment_4 _sweephitresult_4) (CallFunction|HandleHitNormalBump _outvelocity_3 _outoldposition_3 _outnewposition_3 _outnewadjustment_3 _sweephitresult_3 1))
        (if _handledblockinghit_5
          (CallFunction|HandleHitNormalBump _outvelocity_4 _outoldposition_4 _outnewposition_4 _outnewadjustment_4 _sweephitresult_4 2))))
    (:2
      (bind (_handledblockinghit_6 _outvelocity_5 _outoldposition_5 _outnewposition_5 _outnewadjustment_5 _sweephitresult_5) (CallFunction|HandleHitRicochetVelocity InVelocity OldPosition NewLocation Adjustement Hit))
      (if _handledblockinghit_6
        (bind (_handledblockinghit_7 _outvelocity_6 _outoldposition_6 _outnewposition_6 _outnewadjustment_6 _sweephitresult_6) (CallFunction|HandleHitRicochetVelocity _outvelocity_5 _outoldposition_5 _outnewposition_5 _outnewadjustment_5 _sweephitresult_5 1))
        (if _handledblockinghit_7
          (bind _handledblockinghit_8 (CallFunction|HandleHitRicochetVelocity _outvelocity_6 _outoldposition_6 _outnewposition_6 _outnewadjustment_6 _sweephitresult_6 2))
          (if _handledblockinghit_8
            _))))))

(event EventPhysicsRotation
  (params (RuntimeContext :type MMERuntime%20Context%20Structure%20%28by%20ref%29)))

(fn PhysUpdate
  (params (RuntimeContext :type MMERuntime%20Context%20Structure%20%28by%20ref%29) (InVelocity :type Vector%20%28by%20ref%29 :default "0, 0, 0") (InAcceleration :type Vector%20%28by%20ref%29 :default "0, 0, 0"))
  ;[0.1146 0.1146 0.1146] Decelerate Version (snappy turning)
  (Utilities|FlowControl|Sequence
    (:then_1
      (bind _returnvalue (Class|FortMovementModeExtBlueprintLibrary|GetRuntimeDatafromContext (Variables|GetRuntimeContext) "/MistRaven/Gameplay/Ascended/GrimStrutAscended/Blueprints/BpMME_RuntimeData_MistRaven_WraithMovement.BpMME_RuntimeData_MistRaven_WraithMovement_C"))
      (Utilities|FlowControl|Sequence
        (:then_0
          ;;[1 1 1] Kinematic Update Speed
          (LocalVariables|SetInput2DPressedThisFrame (> (Math|Vector2D|Vector2DLengthSquared (Variables|/MistRaven/Gameplay/Ascended/GrimStrutAscended/Blueprints/BpMME_RuntimeData_MistRaven_WraithMovement.BpMME_RuntimeData_MistRaven_WraithMovement_C|GetInputVector _returnvalue)) 0.01))
          (LocalVariables|SetInputVerticalPressedThisFrame (!= (+ (Math|Conversions|ToInteger(Boolean) (Variables|/MistRaven/Gameplay/Ascended/GrimStrutAscended/Blueprints/BpMME_RuntimeData_MistRaven_WraithMovement.BpMME_RuntimeData_MistRaven_WraithMovement_C|GetUpInput _returnvalue)) (* (Math|Conversions|ToInteger(Boolean) (Variables|/MistRaven/Gameplay/Ascended/GrimStrutAscended/Blueprints/BpMME_RuntimeData_MistRaven_WraithMovement.BpMME_RuntimeData_MistRaven_WraithMovement_C|GetDownInput _returnvalue)) -1)) 0))
          (bind _output_get (LocalVariables|SetInputPressedThisFrame (Math|Boolean|ORBoolean (LocalVariables|GetInput2DPressedThisFrame) (LocalVariables|GetInputVerticalPressedThisFrame) (and (Pawn|IsBotControlled (Variables|FortMovementMode_BaseExtLogic|GetOwnerPawn)) (> (Math|Vector|VectorLengthSquared (Variables|/MistRaven/Gameplay/Ascended/GrimStrutAscended/Blueprints/BpMME_RuntimeData_MistRaven_WraithMovement.BpMME_RuntimeData_MistRaven_WraithMovement_C|GetRequestedVelocity _returnvalue)) 0)))))
          (if _output_get
            ;;;[0.4619 1 0.3906] Accelerate
            (LocalVariables|SetNewVelocityLength (+ (Math|Vector|VectorLength (Variables|GetInVelocity)) (* (Variables|Default|GetAccelerationRate) (Class|FortMovementModeExtBlueprintLibrary|GetDeltaTimefromContext (Variables|GetRuntimeContext) (Variables|FortMovementMode_BaseExtLogic|GetbUseDilatedDeltaTime)))))
            (LocalVariables|SetNewAcceleration (Variables|GetInAcceleration))
            (else
              ;;;[1 0.401 0.401 0.5] Deccelerate
              (bind _output_get_2 (LocalVariables|SetNewVelocityLength (+ (Math|Vector|VectorLength (Variables|GetInVelocity)) (Utilities|Operators|Multiply (Class|FortMovementModeExtBlueprintLibrary|GetDeltaTimefromContext (Variables|GetRuntimeContext) (Variables|FortMovementMode_BaseExtLogic|GetbUseDilatedDeltaTime)) (ScalableFloat|GetValueatLevel (Variables|Default|GetDecelerationRate)) -1.0))))))
          (bind _speed (CallFunction|GetLerpedSpeedforThisFrame _output_get_2 (Variables|GetRuntimeContext)))
          (LocalVariables|SetNewVelocityLength _speed))
        (:then_1
          ;;[0.08994 0.04175 0.09896] Snap Velocity To Desired Direction
          (LocalVariables|SetThisFrameScaledUpDownInput_Z (* (Math|Conversions|ToFloat(Integer) (+ (Math|Conversions|ToInteger(Boolean) (Variables|/MistRaven/Gameplay/Ascended/GrimStrutAscended/Blueprints/BpMME_RuntimeData_MistRaven_WraithMovement.BpMME_RuntimeData_MistRaven_WraithMovement_C|GetUpInput _returnvalue)) (* (Math|Conversions|ToInteger(Boolean) (Variables|/MistRaven/Gameplay/Ascended/GrimStrutAscended/Blueprints/BpMME_RuntimeData_MistRaven_WraithMovement.BpMME_RuntimeData_MistRaven_WraithMovement_C|GetDownInput _returnvalue)) -1))) (Math|Float|Clamp(Float) (Math|Float|SelectFloat (ScalableFloat|GetValueatLevel (Variables|UpDownControl|GetUpDownSensitivityMultiplier)) (Math|Curves|GetFloatValue (Variables|UpDownControl|GetUpDownSensitivityCurve) (Math|Float|Clamp(Float) (Math|Float|SafeDivide (Variables|/MistRaven/Gameplay/Ascended/GrimStrutAscended/Blueprints/BpMME_RuntimeData_MistRaven_WraithMovement.BpMME_RuntimeData_MistRaven_WraithMovement_C|GetTotalTimeUpdated _returnvalue) (ScalableFloat|GetValueatLevel (Variables|UpDownControl|GetUpDownDampenFallOffDurationSecs))))) (Variables|UpDownControl|GetUseConstantUpDownSensitivity)))))
          (bind _output_get_3 (LocalVariables|SetInVelDirNormalized (Math|Vector|Normalize (Variables|GetInVelocity))))
          (if (Pawn|IsBotControlled (Variables|FortMovementMode_BaseExtLogic|GetOwnerPawn))
            (LocalVariables|SetDesiredTargetVelDirThisFrame (Math|Vector|Normalize (Variables|/MistRaven/Gameplay/Ascended/GrimStrutAscended/Blueprints/BpMME_RuntimeData_MistRaven_WraithMovement.BpMME_RuntimeData_MistRaven_WraithMovement_C|GetRequestedVelocity _returnvalue)))
            (else
              (bind _output_get_4 (LocalVariables|SetDesiredTargetVelDirThisFrame (Math|Vector|SelectVector (Math|Vector|SelectVector (Math|Vector|RotateVector (Math|Vector|ClampVectorSize (Math|Conversions|ToVector(Vector2D) (Variables|/MistRaven/Gameplay/Ascended/GrimStrutAscended/Blueprints/BpMME_RuntimeData_MistRaven_WraithMovement.BpMME_RuntimeData_MistRaven_WraithMovement_C|GetInputVector _returnvalue) (Math|Float|SelectFloat 0.0 (LocalVariables|GetThisFrameScaledUpDownInput_Z) (Variables|UpDownControl|GetSeparateUpDownInputVector))) 0.0 1.0) (Variables|/MistRaven/Gameplay/Ascended/GrimStrutAscended/Blueprints/BpMME_RuntimeData_MistRaven_WraithMovement.BpMME_RuntimeData_MistRaven_WraithMovement_C|GetControlRotation _returnvalue)) _output_get_3 (select (Variables|UpDownControl|GetSeparateUpDownInputVector) (LocalVariables|GetInput2DPressedThisFrame) (LocalVariables|GetInputPressedThisFrame))) (LocalVariables|GetInVelDirNormalized) (> (Math|Vector|VectorLengthSquared (Math|Vector|SelectVector (Math|Vector|RotateVector (Math|Vector|ClampVectorSize (Math|Conversions|ToVector(Vector2D) (Variables|/MistRaven/Gameplay/Ascended/GrimStrutAscended/Blueprints/BpMME_RuntimeData_MistRaven_WraithMovement.BpMME_RuntimeData_MistRaven_WraithMovement_C|GetInputVector _returnvalue) (Math|Float|SelectFloat 0.0 (LocalVariables|GetThisFrameScaledUpDownInput_Z) (Variables|UpDownControl|GetSeparateUpDownInputVector))) 0.0 1.0) (Variables|/MistRaven/Gameplay/Ascended/GrimStrutAscended/Blueprints/BpMME_RuntimeData_MistRaven_WraithMovement.BpMME_RuntimeData_MistRaven_WraithMovement_C|GetControlRotation _returnvalue)) _output_get_3 (select (Variables|UpDownControl|GetSeparateUpDownInputVector) (LocalVariables|GetInput2DPressedThisFrame) (LocalVariables|GetInputPressedThisFrame)))) 0.001))))))
          (if (Variables|UpDownControl|GetSeparateUpDownInputVector)
            (LocalVariables|SetDesiredTargetVelDirThisFrame (Math|Vector|ClampVectorSize (+ _output_get_4 (Math|Vector|MakeVector 0.0 0.0 (LocalVariables|GetThisFrameScaledUpDownInput_Z))) 0.0 1.0)) ; add back in up/down input after
          )
          (LocalVariables|SetVelocityDragAlpha (Math|Float|Clamp(Float) (Math|Float|SafeDivide (* (ScalableFloat|GetValueatLevel (Variables|Default|GetVelocityDragAngleRateDegrees)) (Class|FortMovementModeExtBlueprintLibrary|GetDeltaTimefromContext (Variables|GetRuntimeContext) (Variables|FortMovementMode_BaseExtLogic|GetbUseDilatedDeltaTime))) (Math|Trig|Acos(Degrees) (Math|Vector|DotProduct (LocalVariables|GetDesiredTargetVelDirThisFrame) (LocalVariables|GetInVelDirNormalized))))))
          (bind _output_get_5 (LocalVariables|SetNewVelocity (* (LocalVariables|GetNewVelocityLength) (Math|Vector|SelectVector (Math|Vector|SlerpVectorToDirection (Math|Vector|SelectVector (LocalVariables|GetInVelDirNormalized) (LocalVariables|GetDesiredTargetVelDirThisFrame) (> (Math|Vector|VectorLengthSquared (LocalVariables|GetInVelDirNormalized)) 0)) (LocalVariables|GetDesiredTargetVelDirThisFrame) (LocalVariables|GetVelocityDragAlpha)) (LocalVariables|GetDesiredTargetVelDirThisFrame) (ScalableFloat|GetValueasBool (Variables|Default|GetEnableVelocityDrag))))))
          (bind _returnvalue_2 (CallFunction|GetZClampSpeedHeight)) ; put as a unction in case we want a faster combined version.
          (if (> _returnvalue_2 0)
            (LocalVariables|SetNewVelocity (Math|Vector|MakeVector (.x _output_get_5) (.y _output_get_5) (Math|Float|Min(Float) (.z _output_get_5) _returnvalue_2)))
            (else
              _))
          (if false
            (if (Utilities|Enum|Equal(Enum) (Networking|GetLocalRole (Variables|FortMovementMode_BaseExtLogic|GetOwnerPawn)) "ROLE_AutonomousProxy")
              (Rendering|Debug|DrawDebugArrow (Variables|FortMovementMode_BaseExtLogic|GetOwnerPawn) (CallFunction|GetDebugVisualBaseLocation "0.000000,0.000000,150.000000") (+ (CallFunction|GetDebugVisualBaseLocation "0.000000,0.000000,150.000000") (* (LocalVariables|GetDesiredTargetVelDirThisFrame) 1e+02)) 1.0 "(R=1.000000,G=0.043382,B=0.000000,A=1.000000)")
              (Rendering|Debug|DrawDebugArrow (Variables|FortMovementMode_BaseExtLogic|GetOwnerPawn) (CallFunction|GetDebugVisualBaseLocation "0.000000,0.000000,200.000000") (+ (CallFunction|GetDebugVisualBaseLocation "0.000000,0.000000,200.000000") (* (Math|Vector|Normalize (LocalVariables|GetNewVelocity)) 1e+02)) 0.0 "(R=0.155667,G=0.000000,B=1.000000,A=1.000000)")
              (else
                _))
            (return (LocalVariables|GetNewVelocity) (LocalVariables|GetNewAcceleration))))
        (:then_2
          (bind _output_get_6 (Variables|Default|SetServerTotalTime (+ (Variables|Default|GetServerTotalTime) (Class|FortMovementModeExtBlueprintLibrary|GetDeltaTimefromContext (Variables|GetRuntimeContext) (Variables|FortMovementMode_BaseExtLogic|GetbUseDilatedDeltaTime))))))))))

(fn UpdateBeforeCharacterMovement
  (params (RuntimeContext :type MMERuntime%20Context%20Structure%20%28by%20ref%29))
  (if (GameplayTags|HasMatchingGameplayTag (Variables|FortMovementMode_BaseExtLogic|GetOwnerPawn) "(TagName=\"MistRaven.GrimStrut.Active\")")
    (return "Active")
    (else
      (return "Ended"))))

(fn DoHoverLogic
  (params (MMERuntimeContext :type MMERuntime%20Context%20Structure%20%28by%20ref%29))
  (if (ScalableFloat|GetValueasBool (Variables|Hover|GetEnableHoverLogic))
    (bind (_outhit _returnvalue) (Collision|LineTraceByChannel :WorldContextObject (Variables|FortMovementMode_BaseExtLogic|GetOwnerPawn) :Start (- (Transformation|GetActorLocation (Variables|FortMovementMode_BaseExtLogic|GetOwnerPawn)) (Math|Vector|MakeVector 0.0 0.0 (Components|Capsule|GetScaledCapsuleHalfHeight (Variables|/Script/Engine.Character|GetCapsuleComponent (Variables|FortMovementMode_BaseExtLogic|GetOwnerPawn))))) :End (- (- (Transformation|GetActorLocation (Variables|FortMovementMode_BaseExtLogic|GetOwnerPawn)) (Math|Vector|MakeVector 0.0 0.0 (Components|Capsule|GetScaledCapsuleHalfHeight (Variables|/Script/Engine.Character|GetCapsuleComponent (Variables|FortMovementMode_BaseExtLogic|GetOwnerPawn))))) (Math|Vector|MakeVector 0.0 0.0 (ScalableFloat|GetValueatLevel (Variables|Hover|GetGroundHoverDistance)))) :TraceChannel "TraceTypeQuery1" :bTraceComplex false :DrawDebugType (select (Variables|Hover|GetDebugHoverLogic) "ForOneFrame" "None") :bIgnoreSelf true :TraceColor "(R=1.000000,G=0.000000,B=0.000000,A=1.000000)" :TraceHitColor "(R=0.000000,G=1.000000,B=0.000000,A=1.000000)" :DrawTime 5.0))
    (if _returnvalue
      (if (Collision|BreakHitResult _outhit)
        (Transformation|SetActorLocation (Variables|FortMovementMode_BaseExtLogic|GetOwnerPawn) (+ (Transformation|GetActorLocation (Variables|FortMovementMode_BaseExtLogic|GetOwnerPawn)) (Math|Vector|MakeVector 0.0 0.0 (Math|Float|Min(Float) (* (Class|FortMovementModeExtBlueprintLibrary|GetDeltaTimefromContext MMERuntimeContext (Variables|FortMovementMode_BaseExtLogic|GetbUseDilatedDeltaTime)) (ScalableFloat|GetValueatLevel (Variables|Hover|GetHoverSpeed))) (- (ScalableFloat|GetValueatLevel (Variables|Hover|GetGroundHoverDistance)) (.Distance (Collision|BreakHitResult _outhit)))))) true false) ; sweep so we collide with world if there is an issue.
      ))))

(fn GetDebugVisualBaseLocation
  (params (Offset :type Vector))
  (return (+ Offset (Transformation|GetActorLocation (Variables|FortMovementMode_BaseExtLogic|GetOwnerPawn)))))

(fn GetLerpedSpeedForThisFrame
  (params (InNewVelocityLength :type Float%20%28double-precision%29) (InRuntimeContext :type MMERuntime%20Context%20Structure))
  (bind _boostfalloffduration (Variables|Default|GetBoostFallOffDuration))
  (bind _returnvalue (ScalableFloat|GetValueatLevel _boostfalloffduration))
  (bind _fastflightspeed_runtime (Variables|Default|GetFastFlightSpeed_Runtime))
  (bind _boostspeedfalloffcurve (Variables|Default|GetBoostSpeedFallOffCurve))
  (bind _slowflightspeed_runtime (Variables|Default|GetSlowFlightSpeed_Runtime))
  (if (GameplayTags|HasMatchingGameplayTag (Variables|FortMovementMode_BaseExtLogic|GetOwnerPawn) "(TagName=\"MistRaven.AllAscendedEquipped\")")
    (Variables|Default|SetFastFlightSpeed_Runtime (ScalableFloat|GetValueasDouble (Variables|Default|GetAscended_FastWraithFlightSpeed)))
    (Variables|Default|SetSlowFlightSpeed_Runtime (ScalableFloat|GetValueasDouble (Variables|Default|GetAscended_SlowWraithFlightSpeed)))
    (else
      (Variables|Default|SetFastFlightSpeed_Runtime (ScalableFloat|GetValueasDouble (Variables|Default|GetFastWraithFlightSpeed)))
      (Variables|Default|SetSlowFlightSpeed_Runtime (ScalableFloat|GetValueasDouble (Variables|Default|GetSlowWraithFlightSpeed)))))
  (bind _returnvalue_2 (Class|FortMovementModeExtBlueprintLibrary|GetRuntimeDatafromContext (Variables|GetInRuntimeContext) "/MistRaven/Gameplay/Ascended/GrimStrutAscended/Blueprints/BpMME_RuntimeData_MistRaven_WraithMovement.BpMME_RuntimeData_MistRaven_WraithMovement_C"))
  (if false
    (Development|PrintString (Variables|FortMovementMode_BaseExtLogic|GetOwnerPawn) (Utilities|String|Append "wraith speed " (Utilities|String|ToString(Float) (Math|Float|Lerp _slowflightspeed_runtime _fastflightspeed_runtime (Math|Curves|GetFloatValue _boostspeedfalloffcurve (Math|Float|Clamp(Float) (Math|Float|SafeDivide (Variables|/MistRaven/Gameplay/Ascended/GrimStrutAscended/Blueprints/BpMME_RuntimeData_MistRaven_WraithMovement.BpMME_RuntimeData_MistRaven_WraithMovement_C|GetTotalTimeUpdated _returnvalue_2) _returnvalue))))))))
  (return (Math|Float|Clamp(Float) (Variables|GetInNewVelocityLength) 0.0 (Math|Float|Lerp _slowflightspeed_runtime _fastflightspeed_runtime (Math|Curves|GetFloatValue _boostspeedfalloffcurve (Math|Float|Clamp(Float) (Math|Float|SafeDivide (Variables|/MistRaven/Gameplay/Ascended/GrimStrutAscended/Blueprints/BpMME_RuntimeData_MistRaven_WraithMovement.BpMME_RuntimeData_MistRaven_WraithMovement_C|GetTotalTimeUpdated _returnvalue_2) _returnvalue)))))))

(fn GetZClampSpeedHeight
  (if (ScalableFloat|GetValueasBool (Variables|ZClamp|GetZSpeedClampEnabled))
    (return (ScalableFloat|GetValueatLevel (select (GameplayTags|HasMatchingGameplayTag (Variables|FortMovementMode_BaseExtLogic|GetOwnerPawn) "(TagName=\"MistRaven.AllAscendedEquipped\")") (Variables|ZClamp|GetZSpeedClamp_Combined) (Variables|ZClamp|GetZSpeedClamp_Normal))))
    (else
      (return -1.0))))

(fn HandleHit_NormalBump
  (params (InVelocity :type Vector) (OldPosition :type Vector) (NewLocation :type Vector) (AdjustmentVector :type Vector) (Hit :type Hit%20Result%20Structure) (DebugIndex :type Integer))
  (bind _normalbumpdistance (Variables|Default|GetNormalBumpDistance))
  (bind _returnvalue (ScalableFloat|GetValueatLevel _normalbumpdistance))
  (bind _ownerpawn (Variables|FortMovementMode_BaseExtLogic|GetOwnerPawn))
  (bind _returnvalue_2 (Transformation|GetActorLocation _ownerpawn))
  (bind (_bblockinghit _binitialoverlap _time _distance _location _impactpoint _normal) (Collision|BreakHitResult Hit))
  (bind _adjustmentvector (Variables|GetAdjustmentVector))
  (bind _returnvalue_3 (Math|Vector|ProjectVectorontoPlane _adjustmentvector _normal))
  (bind _returnvalue_4 (- 1.0 _time))
  (bind _returnvalue_5 (* _returnvalue_3 _returnvalue_4))
  (bind _returnvalue_6 (* _normal _returnvalue))
  (bind _returnvalue_7 (+ _returnvalue_2 _returnvalue_6))
  (if (< _time 1.0)
    ; Normal Bump
    (if (ScalableFloat|GetValueasBool (Variables|Default|GetDebugHitLogicEnabled))
      (Rendering|Debug|DrawDebugArrow (Variables|FortMovementMode_BaseExtLogic|GetOwnerPawn) _returnvalue_2 _returnvalue_7 0.0 "(R=1.000000,G=0.928018,B=0.000000,A=1.000000)" 15.0)
      (else
        _))
    (Transformation|SetActorLocation _ownerpawn _returnvalue_7 true false) ; bump normal distance
    ; Slide!
    (bind _sweephitresult (Transformation|SetActorLocation (Variables|FortMovementMode_BaseExtLogic|GetOwnerPawn) (+ (Transformation|GetActorLocation (Variables|FortMovementMode_BaseExtLogic|GetOwnerPawn)) _returnvalue_5) true false)) ; sweep in case we hit something else.
    (if (ScalableFloat|GetValueasBool (Variables|Default|GetDebugHitLogicEnabled))
      (Rendering|Debug|DrawDebugString (Variables|FortMovementMode_BaseExtLogic|GetOwnerPawn) (Math|Vector|MakeVector 0.0 0.0 (+ 1e+02 (* DebugIndex 50))) (Utilities|String|Append "hit slide " (Utilities|String|ToString(Integer) DebugIndex)) (Variables|FortMovementMode_BaseExtLogic|GetOwnerPawn))
      (else
        _))
    (return true (Transformation|GetVelocity (Variables|FortMovementMode_BaseExtLogic|GetOwnerPawn)) (Variables|GetNewLocation) (Transformation|GetActorLocation (Variables|FortMovementMode_BaseExtLogic|GetOwnerPawn)) _returnvalue_5 _sweephitresult)
    (else
      (return false (Variables|GetInVelocity) (Variables|GetOldPosition) (Variables|GetNewLocation) (Variables|GetAdjustmentVector) (Variables|GetHit)))))

(fn HandleHit_RicochetVelocity
  (params (InVelocity :type Vector) (OldPosition :type Vector) (NewLocation :type Vector) (AdjustmentVector :type Vector) (Hit :type Hit%20Result%20Structure) (DebugIndex :type Integer))
  (bind _invelocity (Variables|GetInVelocity))
  (bind _returnvalue (Math|Vector|GetReflectionVector _invelocity (.Normal (Collision|BreakHitResult (Variables|GetHit)))))
  (bind _ownerpawn (Variables|FortMovementMode_BaseExtLogic|GetOwnerPawn))
  (bind _returnvalue_2 (Transformation|GetActorLocation _ownerpawn))
  (bind _hit (Variables|GetHit))
  (bind (_bblockinghit _binitialoverlap _time _distance _location _impactpoint _normal) (Collision|BreakHitResult _hit))
  (bind _adjustmentvector (Variables|GetAdjustmentVector))
  (bind _returnvalue_3 (Math|Vector|ProjectVectorontoPlane _adjustmentvector _normal))
  (bind _returnvalue_4 (- 1.0 _time))
  (bind _returnvalue_5 (* _returnvalue_3 _returnvalue_4))
  (if (< _time 1.0)
    (if (ScalableFloat|GetValueasBool (Variables|Default|GetDebugHitLogicEnabled))
      (Rendering|Debug|DrawDebugArrow (Variables|FortMovementMode_BaseExtLogic|GetOwnerPawn) _returnvalue_2 (+ _returnvalue_2 (* _normal 5e+01)) 0.0 "(R=1.000000,G=0.928018,B=0.000000,A=1.000000)" 15.0)
      (Rendering|Debug|DrawDebugArrow (Variables|FortMovementMode_BaseExtLogic|GetOwnerPawn) _returnvalue_2 (+ (Transformation|GetActorLocation (Variables|FortMovementMode_BaseExtLogic|GetOwnerPawn)) (* (Math|Vector|Normalize _returnvalue) 5e+01)) 0.0 "(R=1.000000,G=0.000000,B=0.000000,A=1.000000)" 15.0)
      (Rendering|Debug|DrawDebugArrow (Variables|FortMovementMode_BaseExtLogic|GetOwnerPawn) (+ _returnvalue_2 (* (Math|Vector|Normalize (Variables|GetInVelocity)) -5e+01)) _returnvalue_2 0.0 "(R=0.000000,G=0.167813,B=1.000000,A=1.000000)" 15.0)
      (else
        _))
    (Variables|/Script/Engine.MovementComponent|SetVelocity (+ (- (* _returnvalue (ScalableFloat|GetValueatLevel (Variables|Default|GetRicochetSpeedDampening))) (Math|Vector|ProjectVectorontoPlane _returnvalue _normal)) (* (Math|Vector|ProjectVectorontoPlane _returnvalue _normal) (ScalableFloat|GetValueatLevel (Variables|Default|GetRicochetSpeedNormalAlpha)))) (Variables|/Script/Engine.Character|GetCharacterMovement (Variables|FortMovementMode_BaseExtLogic|GetOwnerPawn)))
    ; Slide!
    (bind _sweephitresult (Transformation|SetActorLocation (Variables|FortMovementMode_BaseExtLogic|GetOwnerPawn) (+ (Transformation|GetActorLocation (Variables|FortMovementMode_BaseExtLogic|GetOwnerPawn)) _returnvalue_5) true false)) ; sweep in case we hit something else.
    (if (ScalableFloat|GetValueasBool (Variables|Default|GetDebugHitLogicEnabled))
      (Rendering|Debug|DrawDebugString (Variables|FortMovementMode_BaseExtLogic|GetOwnerPawn) (Math|Vector|MakeVector 0.0 0.0 (+ 1e+02 (* DebugIndex 50))) (Utilities|String|Append "hit slide " (Utilities|String|ToString(Integer) DebugIndex)) (Variables|FortMovementMode_BaseExtLogic|GetOwnerPawn))
      (else
        _))
    (return true (Transformation|GetVelocity (Variables|FortMovementMode_BaseExtLogic|GetOwnerPawn)) (Variables|GetNewLocation) (Transformation|GetActorLocation (Variables|FortMovementMode_BaseExtLogic|GetOwnerPawn)) _returnvalue_5 _sweephitresult)
    (else
      (return false (Variables|GetInVelocity) (Variables|GetOldPosition) (Variables|GetNewLocation) (Variables|GetAdjustmentVector) (Variables|GetHit)))))

(fn HandleHit_Slide
  (params (InVelocity :type Vector) (OldPosition :type Vector) (NewLocation :type Vector) (AdjustmentVector :type Vector) (Hit :type Hit%20Result%20Structure) (DebugIndex :type Integer))
  (bind (_bblockinghit _binitialoverlap _time _distance _location _impactpoint _normal) (Collision|BreakHitResult Hit))
  (bind _returnvalue (Math|Vector|ProjectVectorontoPlane AdjustmentVector _normal))
  (bind _returnvalue_2 (- 1.0 _time))
  (bind _returnvalue_3 (* _returnvalue _returnvalue_2))
  (bind _returnvalue_4 (< _time 1.0))
  ; Fake SlideVectorLogic since actual functions do not seem exposed.
  (if _returnvalue_4
    (bind _sweephitresult (Transformation|SetActorLocation (Variables|FortMovementMode_BaseExtLogic|GetOwnerPawn) (+ NewLocation _returnvalue_3) true false)) ; sweep in case we hit something else.
    (if (ScalableFloat|GetValueasBool (Variables|Default|GetDebugHitLogicEnabled))
      (Rendering|Debug|DrawDebugString (Variables|FortMovementMode_BaseExtLogic|GetOwnerPawn) (Math|Vector|MakeVector 0.0 0.0 (+ 1e+02 (* DebugIndex 50))) (Utilities|String|Append "hit slide " (Utilities|String|ToString(Integer) DebugIndex)) (Variables|FortMovementMode_BaseExtLogic|GetOwnerPawn))
      (else
        _))
    (return _returnvalue_4 (Transformation|GetVelocity (Variables|FortMovementMode_BaseExtLogic|GetOwnerPawn)) NewLocation (Transformation|GetActorLocation (Variables|FortMovementMode_BaseExtLogic|GetOwnerPawn)) _returnvalue_3 _sweephitresult)
    (else
      (return false (Variables|GetInVelocity) (Variables|GetOldPosition) (Variables|GetNewLocation) (Variables|GetAdjustmentVector) (Variables|GetHit)))))
