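; Camera modifier for Ascended GrimStrut that offsets the third-person camera and widens FOV from the player pawn's movement.
; Unless native calculation is used, it derives velocity per frame, interpolating forward, side, vertical and lean offsets, with tactical sprint adjustments.
; A curve-driven lead-in counters initial forward travel from an anchored location, while modifier and ADS alphas scale the effect.
; WindDownAndRemove disables the modifier, waits out its blend time, then removes it.
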(event Custom|WindDownAndRemove
  (CameraModifier|DisableModifier false)
  (Utilities|FlowControl|Delay (Variables|CameraModifier|GetAlphaOutTime))
  (Game|Player|RemoveCameraModifier (Variables|CameraModifier|GetCameraOwner) self))

(fn BlueprintModifyCamera
  (params (DeltaTime :type Float%20%28single-precision%29) (ViewLocation :type Vector :default "0, 0, 0") (ViewRotation :type Rotator :default "0, 0, 0") (FOV :type Float%20%28single-precision%29))
  (bind _velocity (LocalVariables|GetVelocity))
  (if (not (CallFunction|UseNativeCalculation))
    (Utilities|FlowControl|Sequence
      (:then_0
        (Utilities|IsValid (Variables|Default|GetPlayerPawn_Athena)
          (:Is%20Not%20Valid
            (bind _asplayer_pawn_athena (Utilities|Casting|CastToPlayerPawn_Athena (Game|Player|GetPlayerPawn (Utilities|Casting|CastToFortPlayerController (Game|Player|GetOwningPlayerController (Variables|CameraModifier|GetCameraOwner))))))
            (Variables|Default|SetPlayerPawn_Athena _asplayer_pawn_athena)
            (Variables|Default|SetAddedTime (Utilities|Time|GetGameTimeinSeconds))
            (Variables|Default|SetAnchoredLocation (Transformation|GetActorLocation (Variables|Default|GetPlayerPawn_Athena))))))
      (:then_1
        (if (< (- (Utilities|Time|GetGameTimeinSeconds) (Variables|Default|GetAddedTime)) (Variables|Default|GetLeadInTime))
          (Variables|Default|SetLeadIn_ForwardOffset (* (Math|Vector|DotProduct (Math|Vector|GetForwardVector (Variables|GetViewRotation)) (- (Transformation|GetActorLocation (Variables|Default|GetPlayerPawn_Athena)) (Variables|Default|GetAnchoredLocation))) (Utilities|UnNormalizeFloat (Math|Curves|GetFloatValue (Variables|Default|GetForwardPercentCurve) (Utilities|NormalizeFloat (- (Utilities|Time|GetGameTimeinSeconds) (Variables|Default|GetAddedTime)) 0 (Variables|Default|GetLeadInTime) true)) 0 0.8)))))
      (:then_2
        (bind _output_get (LocalVariables|SetVelocity (/ (- (Transformation|GetActorLocation (Variables|Default|GetPlayerPawn_Athena)) (Variables|Default|GetLastFramePosition_)) (Math|Float|SelectFloat 1.0 (Variables|GetDeltaTime) (Math|Float|NearlyEqual(Float) (Variables|GetDeltaTime))))))
        (LocalVariables|SetDirection (Math|Vector|Normalize _output_get))
        (LocalVariables|SetSpeed (Math|Vector|VectorLength _output_get))
        (Variables|Default|SetLastFramePosition_ (Transformation|GetActorLocation (Variables|Default|GetPlayerPawn_Athena))))
      (:then_3
        (Variables|Default|SetADS_Alpha (Math|Interpolation|FInterpTo (Variables|Default|GetADS_Alpha) (Math|Float|SelectFloat 1.0 0.0 (LocalVariables|GetIsShooting)) (Variables|GetDeltaTime) 4.0))
        (LocalVariables|SetForwardSpeed (Math|Vector|DotProduct _velocity (Math|Vector|GetForwardVector (Variables|GetViewRotation))))
        (LocalVariables|SetRightSpeed (Math|Vector|DotProduct _velocity (Math|Vector|GetRightVector (Variables|GetViewRotation))))
        (LocalVariables|SetUpSpeed (Math|Vector|DotProduct _velocity (Math|Vector|GetUpVector (Variables|GetViewRotation))))
        (LocalVariables|SetCurrentStraightness (Math|Float|Absolute(Float) (Math|Vector|DotProduct (Math|Vector|GetForwardVector (Variables|GetViewRotation)) (LocalVariables|GetDirection))))
        (Variables|Default|SetCurrentForward (Math|Interpolation|FInterpTo (Variables|Default|GetCurrentForward) (Utilities|Operators|Add (* (Utilities|RemapFloat (Math|Vector|VectorLengthXY (LocalVariables|GetVelocity)) "(X=1100.000000,Y=1800.000000)" "(X=75.000000,Y=260.000000)" true) 1.15) 0 (Utilities|RemapFloat (Math|Float|Absolute(Float) (LocalVariables|GetUpSpeed)) "(X=0.000000,Y=300.000000)" "(X=0.000000,Y=42.000000)" true) (Math|Float|SelectFloat 1e+01 0.0 (Variables|/Script/FortniteGame.FortPlayerPawn|GetbIsTacticalSprinting (Variables|Default|GetPlayerPawn_Athena)))) (Variables|GetDeltaTime) 3.0))
        (Variables|Default|SetCurrentRight (Math|Interpolation|FInterpTo (Variables|Default|GetCurrentRight) (Utilities|Operators|Multiply (Utilities|RemapFloat (LocalVariables|GetRightSpeed) "(X=-1000.000000,Y=1000.000000)" "(X=-75.000000,Y=75.000000)" true) (Math|Float|SelectFloat 1.1 0.9 (> (Utilities|RemapFloat (LocalVariables|GetRightSpeed) "(X=-1000.000000,Y=1000.000000)" "(X=-75.000000,Y=75.000000)" true) 0)) 1.0 (Utilities|RemapFloat (Math|Float|Absolute(Float) (Variables|Default|GetCurrentForward)) "(X=50.000000,Y=300.000000)" "(X=1.200000,Y=1.500000)" true)) (Variables|GetDeltaTime) 3.0))
        (Variables|Default|SetCurrentUp (Math|Interpolation|FInterpTo (Variables|Default|GetCurrentUp) (+ (Utilities|Operators|Multiply (Utilities|RemapFloat (LocalVariables|GetUpSpeed) "(X=-1000.000000,Y=1000.000000)" "(X=-170.000000,Y=170.000000)" true) (Utilities|RemapFloat (Math|Float|Absolute(Float) (Variables|Default|GetCurrentForward)) "(X=50.000000,Y=400.000000)" "(X=1.000000,Y=1.600000)" true) (Math|Float|SelectFloat (ScalableFloat|GetValueatLevel (Variables|Default|GetSprintingUpOffsetMultiplier)) (ScalableFloat|GetValueatLevel (Variables|Default|GetBaseUpOffsetMultiplier)) (Variables|/Script/FortniteGame.FortPlayerPawn|GetbIsTacticalSprinting (Variables|Default|GetPlayerPawn_Athena))) (Math|Float|SelectFloat 1.35 1.0 (> (Utilities|RemapFloat (LocalVariables|GetUpSpeed) "(X=-1000.000000,Y=1000.000000)" "(X=-170.000000,Y=170.000000)" true) 0))) (Math|Float|SelectFloat -7e+01 -4e+01 (Variables|/Script/FortniteGame.FortPlayerPawn|GetbIsTacticalSprinting (Variables|Default|GetPlayerPawn_Athena)))) (Variables|GetDeltaTime) 4.0))
        (Variables|Default|SetCurrentLean (Math|Interpolation|FInterpTo (Variables|Default|GetCurrentLean) (Utilities|RemapFloat (* (LocalVariables|GetCurrentStraightness) (LocalVariables|GetLeanRight)) "(X=-1.000000,Y=1.000000)" (Math|Vector2D|MakeVector2D (Utilities|RemapFloat (Math|Float|Absolute(Float) (LocalVariables|GetForwardSpeed)) "(X=800.000000,Y=1800.000000)" "(X=40.000000,Y=50.000000)" true) (* (Utilities|RemapFloat (Math|Float|Absolute(Float) (LocalVariables|GetForwardSpeed)) "(X=800.000000,Y=1800.000000)" "(X=40.000000,Y=50.000000)" true) -2.75)) true) (Variables|GetDeltaTime) 4.5))
        (Variables|Default|SetCurrentFOV (Math|Interpolation|FInterpTo (Variables|Default|GetCurrentFOV) (Math|Float|SelectFloat 17.0 23.0 (Variables|/Script/FortniteGame.FortPlayerPawn|GetbIsTacticalSprinting (Variables|Default|GetPlayerPawn_Athena))) (Variables|GetDeltaTime) 2.25))
        (Variables|Default|SetCurrentOffset (Math|Interpolation|FInterpTo (Variables|Default|GetCurrentForward) (Utilities|RemapFloat (LocalVariables|GetSpeed) "(X=-3500.000000,Y=3500.000000)" "(X=-220.000000,Y=220.000000)" true) (Variables|GetDeltaTime) 4.0))
        (Variables|Default|SetCurrentOffset_Vec (Math|Interpolation|VInterpTo (Variables|Default|GetCurrentOffset_Vec) (Utilities|Operators|Add (* (Math|Vector|GetForwardVector (Variables|GetViewRotation)) (Variables|Default|GetCurrentForward)) (* (Math|Vector|GetRightVector (Variables|GetViewRotation)) (+ (Variables|Default|GetCurrentRight) (* (LocalVariables|GetCurrentStraightness) -9e+01))) (* "0.000000,0.000000,1.000000" (- (Variables|Default|GetCurrentUp) 18.0)) (* (Math|Vector|CrossProduct (LocalVariables|GetDirection) "0.000000,0.000000,1.000000") (Variables|Default|GetCurrentLean))) (Variables|GetDeltaTime) 1e+01)))
      (:then_4
        (return (- (Variables|GetViewLocation) (+ (* (Utilities|Operators|Add (* (Variables|Default|GetCurrentOffset_Vec) (- 1.0 (Variables|Default|GetADS_Alpha))) (Utilities|Operators|Multiply (Math|Vector|GetForwardVector (Variables|GetViewRotation)) (- 1.0 (Variables|Default|GetADS_Alpha)) -3e+01) (* (Math|Vector|GetUpVector (Variables|GetViewRotation)) 3e+01) (* (Math|Vector|GetRightVector (Variables|GetViewRotation)) -1e+01)) (Utilities|CalculateAlpha)) (* (Math|Vector|GetForwardVector (Variables|GetViewRotation)) (Variables|Default|GetLeadIn_ForwardOffset)))) (Variables|GetViewRotation) (+ (Variables|GetFOV) (Utilities|Operators|Multiply (Variables|Default|GetCurrentFOV) (Utilities|CalculateAlpha) (- 1.0 (Variables|Default|GetADS_Alpha)))))))))

(fn SetStartingLocation
  (params (StartingLocation :type Vector))
  (Variables|Default|SetStartingLocation StartingLocation))
