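; Gameplay ability that activates the ascended Grim Strut wraith form, a timed mist movement state for the player pawn.
; Activation spends item ammo, applies effects, tags, cooldown, camera changes and a wraith movement extension, while a looping sphere trace breaks weak props and opens doors.
; A montage notify completes the transformation and shrinks the capsule.
; Ending launches the pawn, undoes state with failsafe cleanup, removes depleted items and submits analytics.
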(event Ability|EventActivateAbility
  (bind _returnvalue (CallFunction|GetAvatarasFortPlayerPawn))
  (bind _returnvalue_2 (CallFunction|GetAvatarasFortPlayerPawn))
  (Variables|Default|SetAbilityIsEnding false) ; work around for "active" returning \ntrue when the ability is instanced\nper execution\n(ie it returns true after end ability)
  (CallFunction|StartListeningforHidingPropTags)
  (Class|MistRavenPlayerComponent|BPLogMistRavenStatic "GA GS Activate" (Ability|GetAvatarActorfromActorInfo))
  (bind _returnvalue_3 (Ability|CommitAbility))
  (if (and _returnvalue_3 (not (or (GameplayTags|HasMatchingGameplayTag (CallFunction|GetAvatarasFortPlayerPawn) "(TagName=\"MistRaven.GrimStrut.Active\")") (GameplayTags|HasMatchingGameplayTag (CallFunction|GetAvatarasFortPlayerPawn) "(TagName=\"MistRaven.GrimStrut.Notify_Activated\")"))))
    ; block equipment change until passing back to Player component
    (Tags|BlockAbilitiesWithTags (Ability|GetFortAbilitySystemComponent (Ability|GetAvatarActorfromActorInfo)) (Variables|Default|GetChange%20Equipment%20Tag))
    (Ability|HasAuthority
      (:True
        (if (GameplayTags|HasMatchingGameplayTag (CallFunction|GetAvatarasFortPlayerPawn) "(TagName=\"MistRaven.AllAscendedEquipped\")")
          _)
        ; Server side only manipulates inventory items
        (bind _asfort_inventory_owner_interface (Utilities|Casting|CastToFortInventoryOwnerInterface (Pawn|GetController (CallFunction|GetAvatarasFortPlayerPawn)))
          (:then
            (bind _returnvalue_4 (InventoryManagement|FindExistingItemforDefinition _asfort_inventory_owner_interface "/MistRaven/Gameplay/Ascended/GrimStrutAscended/WID_MistRaven_GrimStrutAscended.WID_MistRaven_GrimStrutAscended"))
            (bind _asfort_world_item (Utilities|Casting|CastToFortWorldItem _returnvalue_4)
              (:then
                (Variables|Default|SetAs%20Fort%20World%20Item _asfort_world_item)
                (Variables|Default|SetAmmoCost (Math|Integer|SelectInt 0 1 (FortWeapon|Ammo|HasInfiniteUses (Ability|GetCurrentSourceWeapon))))
                (bind _output_get (Variables|Default|SetNewAmmoCount (Math|Integer|Max(Integer) (- (Item|GetLoadedAmmo (Variables|Default|GetAs%20Fort%20World%20Item)) (Variables|Default|GetAmmoCost)))))
                (Class|FortWorldItem|SetLoadedAmmo (Variables|Default|GetAs%20Fort%20World%20Item) _output_get)
                (if (== _output_get 0)
                  (Variables|Default|SetDeleteLater true)))
              (:CastFailed
                (bind _returnvalue_5 (InventoryManagement|FindExistingItemforDefinition _asfort_inventory_owner_interface "/MistRaven/Gameplay/Ascended/CombinedAscended/WID_MistRaven_CombinedAscended.WID_MistRaven_CombinedAscended"))
                (bind _asfort_world_item_2 (Utilities|Casting|CastToFortWorldItem _returnvalue_5)
                  (:then
                    (Variables|Default|SetAs%20Fort%20World%20Item _asfort_world_item_2)
                    (bind _output_get_2 (Variables|Default|SetAmmoCost (Math|Integer|SelectInt 0 (ScalableFloat|GetValueasInteger (Variables|Scalables|GetAscendedAmmoCost)) (FortWeapon|Ammo|HasInfiniteUses (Ability|GetCurrentSourceWeapon)))))
                    (if (< (FortWeapon|GetRemainingAmmo (Ability|GetCurrentSourceWeapon)) _output_get_2)
                      (Variables|Default|SetDeleteLater true))
                    (FortWeapon|Ammo|ConsumeSpareAmmo (Ability|GetCurrentSourceWeapon) _output_get_2 true)
                    (if (== (FortWeapon|GetRemainingAmmo (Ability|GetCurrentSourceWeapon)) 0)
                      (Variables|Default|SetDeleteLater true)))
                  (:CastFailed
                    (Class|MistRavenPlayerComponent|BPLogMistRavenStatic "World Item Cast Failed! Not Expected" (Ability|GetAvatarActorfromActorInfo)))))))
          (:CastFailed
            _)))
      (:False
        _))
    (Ability|HasAuthority
      (:True
        (Class|MistRavenPlayerComponent|BPLogMistRavenStatic "GA GS Server Skipped Camera Mod Section" (Ability|GetAvatarActorfromActorInfo) "Verbose"))
      (:False
        (Class|MistRavenPlayerComponent|BPLogMistRavenStatic "GA GS Adding Camera Mod" (Ability|GetAvatarActorfromActorInfo))
        (bind _returnvalue_6 (Game|Player|AddNewCameraModifier (Variables|/Script/Engine.PlayerController|GetPlayerCameraManager (Fortnite|GetFortPlayerControllerfromActor (CallFunction|GetAvatarasFortPlayerPawn))) (Variables|Default|GetCamera%20Modifier%20Class)))
        (bind _ascamera_modifier_grim_strut (Utilities|Casting|CastToCameraModifier_GrimStrut_ _returnvalue_6)
          (:then
            (Variables|Default|SetCamera%20Modifier%20Grim%20Strut _ascamera_modifier_grim_strut))
          (:CastFailed
            (Development|PrintString "ERROR: Couldnt add camera")))))
    (Variables|Default|SetStartTime (Utilities|Time|GetGameTimeinSeconds))
    (Variables|Analytics|SetStartPosition (Transformation|GetActorLocation (CallFunction|GetAvatarasFortPlayerPawn)))
    (Character|UnCrouch (CallFunction|GetAvatarasFortPlayerPawn))
    (Camera|PushCameraMode "/MistRaven/Gameplay/Ascended/GrimStrutAscended/Blueprints/CameraMode_MistRaven_GrimStrut.CameraMode_MistRaven_GrimStrut_C")
    (bind _returnvalue_7 (Ability|ApplyGameplayEffectToOwner "/MistRaven/Gameplay/Ascended/GrimStrutAscended/Blueprints/GE_GrimStrutAscended_DuringActiveWindow.GE_GrimStrutAscended_DuringActiveWindow_C"))
    (Variables|Handles|SetBlockMovementVerbsEffect _returnvalue_7)
    (bind _returnvalue_8 (Ability|ApplyGameplayEffectToOwner "/MistRaven/Gameplay/Ascended/GrimStrutAscended/Blueprints/GE_GrimStrutAscended_NoHealingInStorm.GE_GrimStrutAscended_NoHealingInStorm_C"))
    (Variables|Default|SetStormHealBlock _returnvalue_8)
    (bind _returnvalue_9 (Ability|ApplyGameplayEffectToOwner "/MistRaven/Gameplay/Ascended/GrimStrutAscended/Blueprints/Waterwalking/GE_MistRaven_GrimStrut_WaterWalkingDetect.GE_MistRaven_GrimStrut_WaterWalkingDetect_C"))
    (Variables|Handles|SetWaterWalkingEffect _returnvalue_9)
    (if (not (and (== (ScalableFloat|GetValueatLevel (Variables|Scalables|GetDestructionEnabled)) 0) (== (ScalableFloat|GetValueatLevel (Variables|Scalables|GetAutoDoorOpenEnabled)) 0)))
      (bind _returnvalue_10 (Utilities|Time|SetTimerbyEvent (delegate Custom|DestoryStuffAndOpenDoors) 0.05 true true))
      (Variables|Handles|SetDoorDashHandle _returnvalue_10))
    (Ability|GameplayEffect|AddLooseGameplayTags _returnvalue (Variables|Tags|GetGameplay%20Tags) true)
    (Class|MistRavenPlayerComponent|BPLogMistRavenStatic "GA GS Loose Tags Added V1" (Ability|GetAvatarActorfromActorInfo))
    (if (ScalableFloat|GetValueasBool (Variables|Scalables|GetEnableNativeWraithForm))
      (Class|FortMovementModeExtBlueprintLibrary|ActivateRegisteredMovementModeExtension "/MistRaven/Gameplay/Ascended/GrimStrutAscended/Blueprints/MME_MistRaven_WraithForm_Optimized.MME_MistRaven_WraithForm_Optimized_C" (CallFunction|GetAvatarasFortPlayerPawn))
      (else
        (Class|FortMovementModeExtBlueprintLibrary|ActivateRegisteredMovementModeExtension "/MistRaven/Gameplay/Ascended/GrimStrutAscended/Blueprints/BpMME_MistRaven_WraithMovement.BpMME_MistRaven_WraithMovement_C" (CallFunction|GetAvatarasFortPlayerPawn))))
    (Ability|SendGameplayEventtoActor _returnvalue "(TagName=\"MistRaven.GrimStrut.Ascended.StartCooldown\")")
    (if (GameplayTags|HasMatchingGameplayTag (CallFunction|GetAvatarasFortPlayerPawn) "(TagName=\"MistRaven.TendrilDrain.Ascended.InEquipemtn\")")
      (Variables|Default|SetCooldownLevel 1)
      (else
        (Variables|Default|SetCooldownLevel 0)))
    (if (not (ScalableFloat|GetValueasBoolwithContextParam (CallFunction|GetAvatarasFortPlayerPawn) (Variables|/Script/FortniteGame.FortExperienceSettingsComponent|GetNoCooldowns (Fortnite|Athena|GameState|GetExperienceSettingsComponentfromGameState))))
      (Ability|ApplyGameplayEffectToOwner "/MistRaven/Gameplay/Ascended/GrimStrutAscended/Blueprints/GE_GrimStrutAscended_CooldownFromGE.GE_GrimStrutAscended_CooldownFromGE_C" (Variables|Default|GetCooldownLevel)))
    (Ability|ApplyGameplayEffectToOwner "/MistRaven/Gameplay/Ascended/GrimStrutAscended/Blueprints/GE_GrimStrutAscended_CancelBlockFlamer.GE_GrimStrutAscended_CancelBlockFlamer_C" 0)
    (bind _asb_mist_raven_weapon_ascended (Utilities|Casting|CastToB_MistRaven_WeaponAscended (Ability|GetCurrentSourceWeapon))
      (:then
        (Variables|/MistRaven/Gameplay/Ascended/B_MistRaven_WeaponAscended.B_MistRaven_WeaponAscended_C|SetDeleteItemLater (Variables|Default|GetDeleteLater) _asb_mist_raven_weapon_ascended) ; let weapon know to delete on pickup creation with zero ammo
      )
      (:CastFailed
        _))
    ;[0 0.0358 0.08854] Anim Wait Section
    (bind _notifyname (Animation|Montage|PlayMontage (Variables|/Script/Engine.Character|GetMesh _returnvalue_2) "/MistRaven/Animation/MistRaven_GrimStrutStart.MistRaven_GrimStrutStart" 1.0 0.0 "Default")
      (:then
        (Ability|RemoveGameplayEffectFromOwnerWithGrantedTags "(GameplayTags=((TagName=\"Gameplay.Status.IcyFeet.Trap\"),(TagName=\"Gameplay.Status.IcyFeet.Surface\"),(TagName=\"Gameplay.Status.IcyFeet.Grenade\")))")
        (Ability|Async|WaitGameplayEventToActor (CallFunction|GetAvatarasFortPlayerPawn) "(TagName=\"MistRaven.GrimStrut.Ascended.EndAbility\")"
          (:EventReceived
            (Class|MistRavenPlayerComponent|BPLogMistRavenStatic "End Event Received" (Ability|GetAvatarActorfromActorInfo))
            (Ability|EndAbility))))
      (:OnCompleted
        ;; Completed
        (Class|MistRavenPlayerComponent|BPLogMistRavenStatic "Montage Complete" (Ability|GetAvatarActorfromActorInfo) "Verbose"))
      (:OnBlendOut
        ;; Blend Out
        (Class|MistRavenPlayerComponent|BPLogMistRavenStatic "Montage Blended Out!" (Ability|GetAvatarActorfromActorInfo))
        (if (ScalableFloat|GetValueasBool (Variables|Default|GetBlendOutCancel))
          (CallFunction|AnimPhaseCancelAbility)
          (elif (ScalableFloat|GetValueasBool (Variables|Default|GetBlendOutProceeds))
            (CallFunction|AnimPhaseProceedPassedAnimWait)
            (else
              (Class|MistRavenPlayerComponent|BPLogMistRavenStatic "Blend Out Does Nothing" (Ability|GetAvatarActorfromActorInfo))))))
      (:OnInterrupted
        (if (not (> (ScalableFloat|GetValueasDouble (Variables|Default|GetAnimDoesInterruptedCancel)) 0))
          ;; notify player component ga montage notify start hit
          (CallFunction|AnimPhaseProceedPassedAnimWait :!label _label)))
      (:OnNotifyBegin
        (if (== _notifyname (Variables|Default|GetStartWraithForm)) :!label _label_2
          (if (not (Health|IsDBNO _returnvalue_2))
            (goto _label)
            (else
              _))))
      (:OnNotifyEnd
        (goto _label_2)))
    (CallFunction|AnimPhaseCancelAbility)
    (else
      (Development|PrintString "StillWaitingToBeClear")
      (Ability|ApplyGameplayEffectToOwner "/MistRaven/Gameplay/Ascended/GrimStrutAscended/Blueprints/GE_GrimStrut_CancelAbility.GE_GrimStrut_CancelAbility_C") ; apply cancel GE
    )))

(event Custom|ReleaseEqSwitchBlock
  ;[0.1411 0.15 0.02857 0.5] release equipment block change
  (Tags|UnBlockAbilitiesWithTags (Ability|GetFortAbilitySystemComponent (Ability|GetAvatarActorfromActorInfo)) (Variables|Default|GetChange%20Equipment%20Tag) false) ; Failsafe for unblocking equipment change
)

(event Custom|AnimPhase_ProceedPassedAnimWait
  (bind _returnvalue (CallFunction|GetAvatarasFortPlayerPawn))
  ;[0 0.0358 0.08854] Anim Wait Section
  (CallFunction|ReleaseEqSwitchBlock)
  (if (not (Variables|Default|GetAbilityIsEnding))
    (if (not (Variables|Default|GetAnimNotifyReached))
      (Variables|Default|SetAnimNotifyReached true)
      (Class|MistRavenPlayerComponent|BPLogMistRavenStatic "GA GS Anim Activate Wait Phase Complete" _returnvalue)
      (Ability|GameplayEffect|AddLooseGameplayTags _returnvalue (Variables|Tags|GetGameplay%20Tags%20Activate%20Notify))
      (bind _asplayer_controller (Utilities|Casting|CastToPlayerController (Fortnite|GetControllerfromActor _returnvalue))
        (:then
          (Game|Feedback|ClientPlayForceFeedback _asplayer_controller (Variables|Default|GetForce%20Feedback%20Effect) "None" true false false)
          (CallFunction|HandleCapsuleChange))
        (:CastFailed
          (Class|MistRavenPlayerComponent|BPLogMistRavenStatic "GA GS Failed Controller Check!" (Ability|GetAvatarActorfromActorInfo) "Warning")))
      (else
        (Class|MistRavenPlayerComponent|BPLogMistRavenStatic "GA GS Failed first notify check" (Ability|GetAvatarActorfromActorInfo))))))

(event Custom|camerachange)

(event Custom|StompInput
  (bind _returnvalue (CallFunction|GetAvatarasFortPlayerPawn))
  ; Input reset
  (Math|Integer|IncrementInt (Variables|Default|GetiterCount)
    (:%20%20
      (if (< (Variables|Default|GetiterCount) (Variables|Default|GetmaxCount))
        (bind _ascharacter_movement_component (Utilities|Casting|CastToCharacterMovementComponent (Pawn|GetMovementComponent _returnvalue)))
        (Pawn|Components|PawnMovement|ConsumeInputVector _ascharacter_movement_component)
        (Pawn|Components|PawnMovement|AddInputVector _ascharacter_movement_component (Math|Vector|GetForwardVector (Pawn|GetBaseAimRotation _returnvalue)) true)
        (else
          (Utilities|Time|ClearandInvalidateTimerbyHandle (Variables|Handles|GetinputForceIter)))))))

(event Custom|AnimPhase_CancelAbility
  ;[0 0.0358 0.08854] Anim Wait Section
  (Class|MistRavenPlayerComponent|BPLogMistRavenStatic "GA GS Anim Section Invoking Cancel" (Ability|GetAvatarActorfromActorInfo))
  (Ability|ApplyGameplayEffectToOwner "/MistRaven/Gameplay/Ascended/GrimStrutAscended/Blueprints/GE_GrimStrut_CancelAbility.GE_GrimStrut_CancelAbility_C") ; apply cancel GE
  (CallFunction|ReleaseEqSwitchBlock))

(event Custom|DestoryStuffAndOpenDoors
  (bind _hitwall (Variables|Default|GetHitWall))
  (bind _outhits (Collision|MultiSphereTraceForObjects :Start (+ (Transformation|GetActorLocation (CallFunction|GetAvatarasFortPlayerPawn)) (* (Math|Vector|Normalize (Math|Vector|GetForwardVector (Transformation|GetActorRotation (CallFunction|GetAvatarasFortPlayerPawn)))) 200)) :End (+ (Transformation|GetActorLocation (CallFunction|GetAvatarasFortPlayerPawn)) (* (Math|Vector|Normalize (Math|Vector|GetForwardVector (Transformation|GetActorRotation (CallFunction|GetAvatarasFortPlayerPawn)))) 200)) :Radius 2.5e+02 :ObjectTypes (Variables|Default|GetObjectTypesToCastFor) :bTraceComplex true :DrawDebugType "None" :bIgnoreSelf true :TraceColor "(R=1.000000,G=0.000000,B=0.000000,A=1.000000)" :TraceHitColor "(R=0.000000,G=1.000000,B=0.000000,A=1.000000)" :DrawTime 0.2))
  (for _array_element _outhits
    (bind _asbuilding_prop (Utilities|Casting|CastToBuildingProp (.HitActor (Collision|BreakHitResult _array_element)))
      (:then
        (if (.CanBeDamaged (Class|BPFLAICommonFunctions|GetCanbeDamaged _asbuilding_prop))
          (if (< (Health|GetMaxHealth _asbuilding_prop :!class /Script/FortniteGame.BuildingActor) 9e+01)
            (if (> (ScalableFloat|GetValueatLevel (Variables|Scalables|GetDestructionEnabled)) 0)
              (Game|Damage|ApplyDamage _asbuilding_prop 9e+01 (Pawn|GetController (CallFunction|GetAvatarasFortPlayerPawn)) (CallFunction|GetAvatarasFortPlayerPawn)))
            (else
              _))))
      (:CastFailed
        _))
    (bind _asbuilding_wall (Utilities|Casting|CastToBuildingWall (.HitActor (Collision|BreakHitResult _array_element))))
    (Variables|Default|SetHitWall _asbuilding_wall)
    (bind _asstatic_mesh_component (Utilities|Casting|CastToStaticMeshComponent (.HitComponent (Collision|BreakHitResult _array_element))))
    (bind _returnvalue (BuildingWall|IsDoorComponent _hitwall _asstatic_mesh_component))
    (if (not _returnvalue)
      (bind _returnvalue_2 (BuildingWall|IsDoubleDoorComponent _hitwall _asstatic_mesh_component))
      (if _returnvalue_2
        _))
    (if (Math|Boolean|ANDBoolean (Door|HasBashableDoor _hitwall) (not (Helper|IsDoorOpen _hitwall)) (> (ScalableFloat|GetValueatLevel (Variables|Scalables|GetAutoDoorOpenEnabled)) 0))
      (Openable|OpenActor (Variables|Default|GetHitWall) (Fortnite|GetControllerfromActor (CallFunction|GetAvatarasFortPlayerPawn)) true))))

(event Ability|EventOnEndAbility
  (params (bWasCancelled :type Boolean))
  (bind _returnvalue (CallFunction|GetAvatarasFortPlayerPawn))
  (bind _returnvalue_2 (Fortnite|GetFortPlayerControllerfromActor _returnvalue))
  (bind _returnvalue_3 (CallFunction|GetAvatarasFortPlayerPawn))
  (bind _returnvalue_4 (Transformation|GetVelocity _returnvalue_3))
  (bind _camera_modifier_grim_strut (Variables|Default|GetCamera%20Modifier%20Grim%20Strut))
  (Variables|Default|SetAbilityIsEnding true) ; work around for "active" returning \ntrue when the ability is instanced\nper execution\n(ie it returns true after end ability)
  (Class|MistRavenPlayerComponent|BPLogMistRavenStatic "GA GS End" (Ability|GetAvatarActorfromActorInfo))
  (Variables|Default|SetAnimNotifyReached false)
  (Utilities|FlowControl|Sequence
    (:then_0
      (Ability|HasAuthority
        (:True
          _)
        (:False
          (Utilities|IsValid _camera_modifier_grim_strut
            (:Is%20Valid
              ;[0.15 0.02396 0.02683 0.5] END
              (Class|CameraModifierGrimStrut|WindDownandRemove _camera_modifier_grim_strut))
            (:Is%20Not%20Valid
              (Class|MistRavenPlayerComponent|BPLogMistRavenStatic "GA GS End No Cam Modifier!" (Ability|GetAvatarActorfromActorInfo))))))
      (bind _asplayer_controller (Utilities|Casting|CastToPlayerController (Fortnite|GetControllerfromActor (CallFunction|GetAvatarasFortPlayerPawn)))
        (:then
          (Game|Feedback|ClientStopForceFeedback _asplayer_controller (Variables|Default|GetForce%20Feedback%20Effect) "None"))
        (:CastFailed
          _))
      (Utilities|Time|ClearandInvalidateTimerbyHandle (Variables|Handles|GetDoorDashHandle))
      (Variables|Default|SetMinZ (.z _returnvalue_4))
      (Pawn|Character|LaunchCharacterJump _returnvalue_3 (Math|Vector|MakeVector (.x (* _returnvalue_4 (Math|Curves|GetFloatValue (Variables|Default|GetCurveLaunchCharacterJumpPercentage) (- (Utilities|Time|GetGameTimeinSeconds) (Variables|Default|GetStartTime))))) (.y (* _returnvalue_4 (Math|Curves|GetFloatValue (Variables|Default|GetCurveLaunchCharacterJumpPercentage) (- (Utilities|Time|GetGameTimeinSeconds) (Variables|Default|GetStartTime))))) (Math|Float|Max(Float) (.z (* _returnvalue_4 (Math|Curves|GetFloatValue (Variables|Default|GetCurveLaunchCharacterJumpPercentage) (- (Utilities|Time|GetGameTimeinSeconds) (Variables|Default|GetStartTime))))) (Variables|Default|GetMinZ))) true false false true)
      (Ability|RemoveGameplayEffectFromOwnerWithHandle (Variables|Default|GetStormHealBlock))
      (Ability|RemoveGameplayEffectFromOwnerWithHandle (Variables|Handles|GetBlockMovementVerbsEffect))
      (Ability|RemoveGameplayEffectFromOwnerWithHandle (Variables|Handles|GetWaterWalkingEffect))
      (if (> (ScalableFloat|GetValueatLevel (Variables|Scalables|GetFallImmunityEnabled)) 0)
        (Ability|ApplyGameplayEffectToOwner "/Game/Athena/Items/Parents/FallImmunity/GE_Athena_TillLandFallDamageImmunity_NoFX_Grant_Parent.GE_Athena_TillLandFallDamageImmunity_NoFX_Grant_Parent_C")
        (else
          (Pawn|Character|ResetFallingHeight (CallFunction|GetAvatarasFortPlayerPawn))))
      (CallFunction|HandleCapsuleReset)
      (Camera|PopCameraMode)
      (Ability|GameplayEffect|RemoveLooseGameplayTags (CallFunction|GetAvatarasFortPlayerPawn) (Variables|Tags|GetGameplay%20Tags%20Activate%20Notify))
      ;; failsafe release equipment block change
      (CallFunction|ReleaseEqSwitchBlock)
      (if (Variables|Default|GetDeleteLater)
        (Inventory|RemoveItemfromInventoryOwner _returnvalue_2 "/MistRaven/Gameplay/Ascended/GrimStrutAscended/WID_MistRaven_GrimStrutAscended.WID_MistRaven_GrimStrutAscended")
        (Inventory|RemoveItemfromInventoryOwner _returnvalue_2 "/MistRaven/Gameplay/Ascended/CombinedAscended/WID_MistRaven_CombinedAscended.WID_MistRaven_CombinedAscended")))
    (:then_1
      ; Unconditional clean up, no logic here, only dirct node calls
      (CallFunction|ForceTagCleanUp)
      (CallFunction|ForcePlayerComponentCleanUp)
      (CallFunction|SubmitAnalytics))))

(event Custom|ForceTagCleanUp
  (bind _gameplay_tags (Variables|Tags|GetGameplay%20Tags))
  (bind _gameplay_tags_activate_notify (Variables|Tags|GetGameplay%20Tags%20Activate%20Notify))
  ;[1 0.25 0.25] Clean Up Loose Tags
  (if (ScalableFloat|GetValueasBool (Variables|Default|GetGAManagesTags))
    (bind _outfortplayerpawn (CallFunction|GetAvatarasValidFortPlayerPawn)
      (:ValidFortPlayerPawn
        (Utilities|FlowControl|Sequence
          (:then_0
            (if (GameplayTags|HasAnyMatchingGameplayTags _outfortplayerpawn _gameplay_tags)
              (Ability|GameplayEffect|RemoveLooseGameplayTags _outfortplayerpawn _gameplay_tags true) ; replicat etag should match its add node
              (Class|MistRavenPlayerComponent|BPLogMistRavenStatic "GA GS GameplayTags detected and removed" _outfortplayerpawn)))
          (:then_1
            (if (GameplayTags|HasAnyMatchingGameplayTags _outfortplayerpawn _gameplay_tags_activate_notify)
              (Ability|GameplayEffect|RemoveLooseGameplayTags _outfortplayerpawn _gameplay_tags_activate_notify) ; replicat etag should match its add node
              (Class|MistRavenPlayerComponent|BPLogMistRavenStatic "GA GS Notify tags detected and removed" _outfortplayerpawn)))))
      (:AvatarCastFailed
        (Class|MistRavenPlayerComponent|BPLogMistRavenStatic "GA GS failed to get avatar" _outfortplayerpawn "Warning")))))

(event Custom|ForcePlayerComponentCleanUp
  ;[1 0.07013 0] Clean Up Player Component
  (if (ScalableFloat|GetValueasBool (Variables|Default|GetGAEndsWFActive))
    (bind _outfortplayerpawn (CallFunction|GetAvatarasValidFortPlayerPawn)
      (:ValidFortPlayerPawn
        (Utilities|IsValid (Actor|GetComponentbyClass _outfortplayerpawn "/MistRaven/B_MistRaven_PlayerComponent.B_MistRaven_PlayerComponent_C")
          (:Is%20Valid
            (Utilities|FlowControl|Sequence
              (:then_0
                (if (or (Variables|/Script/MistRavenRuntime.MistRavenPlayerComponent|GetbRepWFActive (Actor|GetComponentbyClass _outfortplayerpawn "/MistRaven/B_MistRaven_PlayerComponent.B_MistRaven_PlayerComponent_C")) (ScalableFloat|GetValueasBool (Variables|Default|GetGAEndsWFActiveEvenIfFalse)))
                  (Class|MistRavenPlayerComponent|BPLogMistRavenStatic "GA clearing WFActive")
                  (Class|MistRavenPlayerComponent|SetWFActiveVariable (Actor|GetComponentbyClass _outfortplayerpawn "/MistRaven/B_MistRaven_PlayerComponent.B_MistRaven_PlayerComponent_C") false)
                  (else
                    (Class|MistRavenPlayerComponent|BPLogMistRavenStatic "GA  WFActive already clear" 0 "Verbose"))))
              (:then_1
                (if (ScalableFloat|GetValueasBool (Variables|Default|GetGAEndsMaterialVariable))
                  (Variables|/MistRaven/B_MistRaven_PlayerComponent.B_MistRaven_PlayerComponent_C|SetbEnableMistOverrideMaterial false (Actor|GetComponentbyClass _outfortplayerpawn "/MistRaven/B_MistRaven_PlayerComponent.B_MistRaven_PlayerComponent_C")) ; make sure material cleans up
                ))))
          (:Is%20Not%20Valid
            (Class|MistRavenPlayerComponent|BPLogMistRavenStatic "GA GS failed to GS player  component" 0 "Warning") ; put under verbose since the tag check also does this\nand we likely will have both turned on.
          )))
      (:AvatarCastFailed
        (Class|MistRavenPlayerComponent|BPLogMistRavenStatic "GA GS failed to get avatar" 0 "Verbose") ; put under verbose since the tag check also does this\nand we likely will have both turned on.
      ))))

(event Custom|StartListeningForHidingPropTags
  ;[1 0 0 0.5] Hiding Prop Cancel
  (bind _outfortplayerpawn (CallFunction|GetAvatarasValidFortPlayerPawn)
    (:ValidFortPlayerPawn
      (Ability|Tasks|WaitGameplayTagAddToActor _outfortplayerpawn "(TagName=\"Pawn.HidePlayerPawn.HidingProp\")"
        (:Added
          (Ability|ApplyGameplayEffectToOwner "/MistRaven/Gameplay/Ascended/GrimStrutAscended/Blueprints/GE_GrimStrut_CancelAbility.GE_GrimStrut_CancelAbility_C") ; apply cancel GE
          (CallFunction|ReleaseEqSwitchBlock))) ; if you get placed in hiding prop somehow, t\nhen exit wraith form to avoid invul state exploit.
    )
    (:AvatarCastFailed
      (Class|MistRavenPlayerComponent|BPLogMistRavenStatic "cannot get pawn to wait for hiding prop tags" (Ability|GetPlayerControllerfromActorInfo)))))

(fn HandleCapsuleChange
  (bind _returnvalue (CallFunction|GetAvatarasFortPlayerPawn))
  (bind _returnvalue_2 (CallFunction|GetAvatarasFortPlayerPawn))
  (bind _capsulecomponent (Variables|/Script/Engine.Character|GetCapsuleComponent _returnvalue))
  (if (ScalableFloat|GetValueasBool (Variables|Default|GetEnableCapsuleChanges))
    (Ability|HasAuthority
      (:True
        (Variables|Default|SetReturnToRadius (Variables|/Script/Engine.CapsuleComponent|GetCapsuleRadius _capsulecomponent))
        (Variables|Default|SetReturnToHalfHeight (Variables|/Script/Engine.CapsuleComponent|GetCapsuleHalfHeight _capsulecomponent))
        (Pawn|SetCapsuleRadiusFortPawn _returnvalue_2 3e+01)
        (Pawn|SetCapsuleHalfHeightFortPawn _returnvalue_2 5e+01)
        (Class|MistRavenPlayerComponent|BPLogMistRavenStatic "GA GS Capsule Change" _returnvalue_2 "Warning") ; put under verbose since the tag check also does this\nand we likely will have both turned on.
      )
      (:False
        _))
    (return)))

(fn HandleCapsuleReset
  (bind _returnvalue (CallFunction|GetAvatarasFortPlayerPawn))
  (if (ScalableFloat|GetValueasBool (Variables|Default|GetEnableCapsuleChanges))
    (Ability|HasAuthority
      (:True
        (Pawn|SetCapsuleRadiusFortPawn _returnvalue (Variables|Default|GetReturnToRadius))
        (Pawn|SetCapsuleHalfHeightFortPawn _returnvalue (Variables|Default|GetReturnToHalfHeight))
        (Class|MistRavenPlayerComponent|BPLogMistRavenStatic "GA GS Capsule Reset" _returnvalue))
      (:False
        _))
    (return)))

(fn SubmitAnalytics
  (bind _returnvalue (CallFunction|GetAvatarasFortPlayerPawn))
  (bind _returnvalue_2 (CallFunction|GetAvatarasFortPlayerPawn))
  (bind _returnvalue_3 (Actor|GetComponentbyClass _returnvalue_2 "/MistRaven/B_MistRaven_PlayerComponent.B_MistRaven_PlayerComponent_C"))
  (bind _returnvalue_4 (Development|GetConsoleVariableBoolValue "MistRaven.Analytics.EnableWraithFormAnalytics"))
  (if _returnvalue_4
    (Ability|HasAuthority
      (:True
        (if (not (Replay|GetIsPlayingReplay))
          (Utilities|FlowControl|Sequence
            (:then_0
              (bind _returnvalue_5 (PlayerInfo|GetDebugStringforUniqueId (Default|GetGameAccountId (Fortnite|GetFortPlayerControllerfromActor (CallFunction|GetAvatarasFortPlayerPawn)))))
              (bind _returnvalue_6 (Analytics|ConstructAnalyticsAttribute "AccountID" _returnvalue_5))
              (Utilities|Array|Add (LocalVariables|GetEventAttributes) _returnvalue_6)
              (bind _returnvalue_7 (Analytics|ConstructAnalyticsAttribute :AttributeName "MatchTime" :IntValue -1 :floatValue (Fortnite|GetMatchTime) :BoolValue false))
              (Utilities|Array|Add (LocalVariables|GetEventAttributes) _returnvalue_7)
              (bind _returnvalue_8 (Analytics|ConstructAnalyticsAttribute "StartPosition" (Utilities|String|ToString(Vector) (Variables|Analytics|GetStartPosition))))
              (Utilities|Array|Add (LocalVariables|GetEventAttributes) _returnvalue_8)
              (bind _returnvalue_9 (Analytics|ConstructAnalyticsAttribute "EndPosition" (Utilities|String|ToString(Vector) (Transformation|GetActorLocation _returnvalue))))
              (Utilities|Array|Add (LocalVariables|GetEventAttributes) _returnvalue_9)
              (bind _returnvalue_10 (Analytics|ConstructAnalyticsAttribute :AttributeName "Health" :IntValue -1 :floatValue (FortPawn|GetHealth _returnvalue) :BoolValue false))
              (Utilities|Array|Add (LocalVariables|GetEventAttributes) _returnvalue_10)
              (bind _returnvalue_11 (Analytics|ConstructAnalyticsAttribute :AttributeName "Shield" :IntValue -1 :floatValue (FortPawn|GetShield _returnvalue) :BoolValue false))
              (Utilities|Array|Add (LocalVariables|GetEventAttributes) _returnvalue_11)
              (bind _returnvalue_12 (Analytics|ConstructAnalyticsAttribute :AttributeName "Ammo" :IntValue (Variables|Default|GetNewAmmoCount) :floatValue -1.0 :BoolValue false))
              (Utilities|Array|Add (LocalVariables|GetEventAttributes) _returnvalue_12)
              (Utilities|IsValid _returnvalue_3
                (:Is%20Valid
                  (bind _returnvalue_13 (Analytics|ConstructAnalyticsAttribute "SessionEndReason" (Utilities|String|SelectString "Cancelled" "Expired" (< (- (Utilities|Time|GetGameTimeinSeconds) (Variables|Default|GetStartTime)) (Variables|/MistRaven/B_MistRaven_PlayerComponent.B_MistRaven_PlayerComponent_C|GetOriginalGameDataDuration _returnvalue_3)))))
                  (Utilities|Array|Add (LocalVariables|GetEventAttributes) _returnvalue_13))))
            (:then_1
              (Analytics|FireAnalyticsEventBlueprintArray "MystFormSessionEnd" (LocalVariables|GetEventAttributes)))))))))

(fn UpdateMeshCollision
  (params (Ignore :type Boolean) (SkeletalMesh :type Skeletal%20Mesh%20Component%20Object%20Reference))
  (bind _skeletalmesh (Variables|GetSkeletalMesh))
  (Variables|Default|SetIgnore Ignore)
  (Variables|Default|SetSkeletal%20Mesh SkeletalMesh)
  (Collision|SetCollisionResponsetoChannel _skeletalmesh "ECC_GameTraceChannel7" (select (Variables|Default|GetIgnore) "ECR_Ignore" "ECR_Ignore"))
  (Collision|SetCollisionResponsetoChannel _skeletalmesh "ECC_GameTraceChannel14" (select (Variables|Default|GetIgnore) "ECR_Ignore" "ECR_Ignore")))
