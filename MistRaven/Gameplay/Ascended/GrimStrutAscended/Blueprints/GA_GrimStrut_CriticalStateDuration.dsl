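; GA_GrimStrut_CriticalStateDuration is a gameplay ability for the Grim Strut ascended character, named for the duration of its Wraith Form critical state.
; Its activate and end events only print a colored on-screen message when the Debug variable is set, marking when the state starts and stops.
; The graph holds no other logic and never ends the ability itself.
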(event Ability|EventActivateAbility
  (Utilities|FlowControl|Sequence
    (:then_0
      (if (Variables|Default|GetDebug)
        (Development|PrintString "WRAITH FORM CRITICAL STATE ACTIVE" true true "(R=0.000000,G=1.000000,B=0.000000,A=1.000000)" 5.0)))))

(event Ability|EventOnEndAbility
  (params (bWasCancelled :type Boolean))
  (Utilities|FlowControl|Sequence
    (:then_0
      (if (Variables|Default|GetDebug)
        (Development|PrintString "DEACTIVATE WRAITH FORM CRITICAL STATE ACTIVE" true true "(R=1.000000,G=0.000000,B=0.000000,A=1.000000)" 5.0)))))
