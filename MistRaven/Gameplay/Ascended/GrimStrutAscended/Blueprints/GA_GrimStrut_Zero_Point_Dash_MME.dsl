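; Gameplay ability for the MistRaven Ascended Grim Strut zero point dash, chaining the dash into wraith movement.
; On activation it prints a debug string, picks longer or standard dash and guided motion MME classes from a TendrilDrain equipment tag, then defers to the parent.
; On end, cancelled or not, it activates either the optimized or Blueprint wraith movement extension, chosen by EnableNativeWraithForm, without calling the parent event.
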(event Ability|EventActivateAbility
  (Development|PrintString "Timing")
  (if (GameplayTags|HasMatchingGameplayTag (CallFunction|GetAvatarasFortPlayerPawn) "(TagName=\"MistRaven.TendrilDrain.Ascended.InEquipemtn\")")
    (Variables|Default|SetDash%20MME%20Class "/MistRaven/Gameplay/Ascended/GrimStrutAscended/Blueprints/GrimStrut_ZeroPointDash_MME_Longer.GrimStrut_ZeroPointDash_MME_Longer_C")
    (Variables|Default|SetGuided%20Motion%20MME%20Class "/MistRaven/Gameplay/Ascended/GrimStrutAscended/Blueprints/FMM_GuidedMotion_GrimStrut_ZeroPointDash_Longer.FMM_GuidedMotion_GrimStrut_ZeroPointDash_Longer_C")
    (else
      (Variables|Default|SetDash%20MME%20Class "/MistRaven/Gameplay/Ascended/GrimStrutAscended/Blueprints/GrimStrut_ZeroPointDash_MME.GrimStrut_ZeroPointDash_MME_C")
      (Variables|Default|SetGuided%20Motion%20MME%20Class "/MistRaven/Gameplay/Ascended/GrimStrutAscended/Blueprints/FMM_GuidedMotion_GrimStrut_ZeroPointDash.FMM_GuidedMotion_GrimStrut_ZeroPointDash_C")))
  (|CallParentFunction))

(event Ability|EventOnEndAbility
  (params (bWasCancelled :type Boolean))
  (if (ScalableFloat|GetValueasBool (Variables|Default|GetEnableNativeWraithForm))
    (Class|FortMovementModeExtBlueprintLibrary|ActivateRegisteredMovementModeExtension "/MistRaven/Gameplay/Ascended/GrimStrutAscended/Blueprints/MME_MistRaven_WraithForm_Optimized.MME_MistRaven_WraithForm_Optimized_C" (CallFunction|GetAvatarasFortPlayerPawn))
    (else
      (Class|FortMovementModeExtBlueprintLibrary|ActivateRegisteredMovementModeExtension "/MistRaven/Gameplay/Ascended/GrimStrutAscended/Blueprints/BpMME_MistRaven_WraithMovement.BpMME_MistRaven_WraithMovement_C" (CallFunction|GetAvatarasFortPlayerPawn)))))
