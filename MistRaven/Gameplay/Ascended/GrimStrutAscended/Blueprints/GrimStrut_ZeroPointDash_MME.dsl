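; GrimStrut_ZeroPointDash_MME is a Blueprint for Grim Strut's Zero Point Dash in the Ascended gameplay content, apparently a movement extension.
; It defines BP_UpdateBeforeCharacterMovement, a per-update hook that receives DeltaTime before character movement runs.
; The hook contains no movement logic and only returns Updating, so the file does not show how the dash itself moves the character.
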(fn BP_UpdateBeforeCharacterMovement
  (params (DeltaTime :type Float%20%28single-precision%29))
  (return "Updating"))
